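# Animation Blueprint for the Alien Summer backpack cosmetic that procedurally animates a floating, spinning UFO and adjusts pack placement during skydiving and parachuting.
# Update event derives a sine bob and opposite-direction inner and outer spins from real time, feeding additive Modify Bone nodes.
# Low-weight Anim Dynamics adds linear spring sway to the float joint, damping component acceleration response in the front end.
# Skydive and parachute flags become alphas that offset the pack and root bones.
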AnimBlueprint /BRCosmetics/Accessories/FORT_Backpacks/M_MED_Alien_Summer/Meshes/M_MED_Alien_Summer_Pack_AnimBP
target_skeleton=&/M_MED_Alien_Summer_Pack_Skeleton
graphs [
 AnimGraph {nodes [
   n1 = "Transform (Modify) Bone - Bone: dyn_float_jnt"(
    ComponentPose=Local To Component(LocalPose=Input Pose())
    Alpha=1
    Translation=Get translateUFO() -> (translateUFO:Vector)
    Rotation="0,0,0"
    Scale="1,1,1"
   ) {translationMode=BMM_Additive, translationSpace=BCS_BoneSpace, alphaBoolBlend {blendOption=Linear}, boneToModify {boneName=dyn_float_jnt}}
   n2 = "Transform (Modify) Bone - Bone: dyn_innerRotation_jnt"(ComponentPose=@n1, Alpha=1, Translation="0,0,0", Rotation=Get spinInner() -> (spinInner:Rotator), Scale="1,1,1") {rotationMode=BMM_Additive, rotationSpace=BCS_BoneSpace, alphaBoolBlend {blendOption=Linear}, boneToModify {boneName=dyn_innerRotation_jnt}}
   n3 = "Transform (Modify) Bone - Bone: dyn_outerRotation_jnt"(ComponentPose=@n2, Alpha=1, Translation="0,0,0", Rotation=Get spinOuter() -> (spinOuter:Rotator), Scale="1,1,1") {rotationMode=BMM_Additive, rotationSpace=BCS_BoneSpace, alphaBoolBlend {blendOption=Linear}, boneToModify {boneName=dyn_outerRotation_jnt}}
   n4 = Anim Dynamics - Bone: dyn_float_jnt!(
    ComponentPose=@n3
    Alpha=0.3
    GravityScale:float=0.2
    ComponentLinearAccScale:Vector=SelectVector(A="0.02,0.02,0.02", B="0.23,0.25,0.19", bPickA=Get bIsInFrontEnd())
    ExternalForce="0,0,0"
   ) -> (Pose) {
    bGravityOverrideInSimSpace=true
    bLinearSpring=true
    gravityScale=0.2
    linearDampingOverride=0.8
    linearSpringConstant=60
    alphaBoolBlend {blendOption=Linear}
    boundBone {boneName=dyn_float_jnt}
    componentLinearAccScale=<0.23 0.25 0.19>
    physicsBodyDefinitions [{
      sphereCollisionRadius=0
      boundBone {boneName=dyn_float_jnt}
      boxExtents=<1 1 1>
      constraintSetup {angularLimitsMax=<0 0 0>, angularLimitsMin=<0 0 0>, angularTarget=<0 0 0>, linearAxesMax=<5 0 10>, linearAxesMin=<-5 0 -10>}
      localJointOffset=<0 0 0>
     }]
   }
   n5 = "Transform (Modify) Bone - Bone: pack"(ComponentPose=@n4, Alpha=Get isSkydivingAlpha() -> (isSkydivingAlpha:double), Translation="0,-6,-5", Rotation="0,0,0") {
    rotationSpace=BCS_BoneSpace
    translationMode=BMM_Additive
    alphaBoolBlend {blendOption=Linear}
    alphaScaleBiasClamp {bInterpResult=true}
    boneToModify {boneName=pack}
    translation {y=-6, z=-5}
   }
   n6 = "Transform (Modify) Bone - Bone: root"(ComponentPose=@n5, Alpha=Get isParachuteDeployed() -> (isParachuteDeployed:double), Translation="0,2,5.5", Rotation="0,0,0") {
    rotationSpace=BCS_BoneSpace
    translationMode=BMM_Additive
    alphaBoolBlend {blendOption=Linear}
    alphaScaleBiasClamp {bInterpResult=true, interpSpeedDecreasing=20, interpSpeedIncreasing=20}
    boneToModify {boneName=root}
    translation {y=2, z=5.5}
   }
   Output Pose(Result=Component To Local(ComponentPose=@n6)) {layerGroup=None}
  ]}
 EventGraph {nodes [
   Event BlueprintUpdateAnimation()
   > n1 = Sequence!(execute) -> (then_0, then_1, then_2:Exec, then_3:Exec, then_4:Exec)
   n2 = GetRealTimeSeconds()
   n3 = float * float(A="Sin (Radians)"(A=float * float(A=@n2, B=2)), B=1.5)
   > Set translateUFO(translateUFO:Vector=MakeVector(Y=float + float(A=@n3, B=-2) {is=title K2Node_CommutativeAssociativeBinaryOperator})) -> (Output_Get:Vector)
   n4 = GetRealTimeSeconds()
   Set spinOuter(execute=@n1.then_1, spinOuter:Rotator=MakeRotator(Pitch=float * float(A=@n4, B=-50))) -> (Output_Get:Rotator)
   n5 = GetRealTimeSeconds()
   Set spinInner(execute=@n1.then_2, spinInner:Rotator=MakeRotator(Pitch=float * float(A=@n5, B=110))) -> (Output_Get:Rotator)
   Set isSkydivingAlpha(execute=@n1.then_3, isSkydivingAlpha:double="To Float (Boolean)"(InBool=Get bIsSkyDiving())) -> (Output_Get:double)
   Set isParachuteDeployed(execute=@n1.then_4, isParachuteDeployed:double="To Float (Boolean)"(InBool=Get bIsParachuteOpened())) -> (Output_Get:double)
   TryGetPawnOwner()
  ]}
]
variables [
 isParachuteDeployed {type=number}
 isSkydivingAlpha {type=number}
 spinInner {type=Rotator}
 spinOuter {type=Rotator}
 translateUFO {type=Vector}
]
